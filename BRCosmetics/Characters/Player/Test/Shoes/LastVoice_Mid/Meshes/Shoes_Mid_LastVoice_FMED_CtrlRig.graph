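# Control Rig for the LastVoice mid-top shoe mesh (FMED variant) in the player cosmetics Test folder, posing a lower-body skeleton through five controls.
# A Sequence first offsets the root bone by root_ctrl and constrains the left leg, then constrains the right.
# Parent constraints bind each calf bone, not the foot, to its foot control, and calf_twist_01 to a twist control nested beneath it.
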ControlRigBlueprint /BRCosmetics/Characters/Player/Test/Shoes/LastVoice_Mid/Meshes/Shoes_Mid_LastVoice_FMED_CtrlRig
graphs [
 RigVMFunctionLibrary {links [], nodes []}
 RigVMModel {nodes [
   RigUnit_BeginExecution = &RigUnit_BeginExecution()
   RigVMFunction_Sequence = &RigVMFunction_Sequence(ExecuteContext=@RigUnit_BeginExecution.ExecutePin)
   RigUnit_GetTransform = &RigUnit_GetTransform(Item=(Type=Control,Name="root_ctrl"), bInitial=False)
   RigUnit_OffsetTransformForItem = &RigUnit_OffsetTransformForItem(
    Item=(Type=Bone,Name="root")
    Weight=1
    bPropagateToChildren=True
    OffsetTransform=@RigUnit_GetTransform.Transform
    ExecutePin=@RigVMFunction_Sequence.A
   )
   > &RigUnit_ParentConstraint(
    AdvancedSettings=(InterpolationType=Average,RotationOrderForFilter=XZY)
    Child=(Type=Bone,Name="calf_l")
    Parents=((Item=(Type=Control,Name="foot_l_ctrl"),Weight=1),(Item=(Type=Bone,Name="None"),Weight=1))
    Weight=1
   )
   > &RigUnit_ParentConstraint(
    AdvancedSettings=(InterpolationType=Average,RotationOrderForFilter=XZY)
    Child=(Type=Bone,Name="calf_twist_01_l")
    Parents=((Item=(Type=Control,Name="calf_twist_01_l_ctrl"),Weight=1),(Item=(Type=Bone,Name="None"),Weight=1))
    Weight=1
   )
   &RigUnit_ParentConstraint(
    AdvancedSettings=(InterpolationType=Average,RotationOrderForFilter=XZY)
    Child=(Type=Bone,Name="calf_r")
    Parents=((Item=(Type=Control,Name="foot_r_ctrl"),Weight=1),(Item=(Type=Bone,Name="None"),Weight=1))
    Weight=1
    ExecutePin=@RigVMFunction_Sequence.B
   )
   > &RigUnit_ParentConstraint(
    AdvancedSettings=(InterpolationType=Average,RotationOrderForFilter=XZY)
    Child=(Type=Bone,Name="calf_twist_01_r")
    Parents=((Item=(Type=Control,Name="calf_twist_01_r_ctrl"),Weight=1),(Item=(Type=Bone,Name="None"),Weight=1))
    Weight=1
   )
  ]}
]
hierarchy [
 root {loc=<0 0 0>, rot=<0 0 0 1>}
 pelvis {parent=root, loc=<0 -0.5301 98.0571>, rot=<0.0127 0.707 -0.0127 -0.707>}
 thigh_l {parent=pelvis, loc=<-1.0855 -0.2109 -8.7053>, rot=<0.0009 0.0355 0.0155 -0.9993>}
 calf_l {parent=thigh_l, loc=<-42.5284 -0.0019 -0.7297>, rot=<0.0006 -0.0161 -0.0391 -0.9991>}
 foot_l {parent=calf_l, loc=<-46.7402 -2.5014 -0.592>, rot=<0.0427 -0.0238 0.0353 -0.9982>}
 ball_l {parent=foot_l, loc=<-3.5658 10.8066 -0.1544>, rot=<0.0001 -0.0001 -0.7063 -0.7079>}
 calf_twist_02_l {parent=calf_l, loc=<-19.1546 -1.7943 -0.2638>, rot=<0.0237 0.0095 -0.0035 -0.9997>}
 calf_twist_01_l {parent=calf_l, loc=<-35.8483 -1.9197 -0.5794>, rot=<0.0237 0.0095 -0.0035 -0.9997>}
 thigh_r {parent=pelvis, loc=<-1.0855 -0.2109 8.7053>, rot=<0.0355 -0.0009 0.9993 0.0155>}
 calf_r {parent=thigh_r, loc=<42.5284 0.0019 0.7297>, rot=<0.0006 -0.0161 -0.0391 -0.9991>}
 foot_r {parent=calf_r, loc=<46.7402 2.5014 0.592>, rot=<0.0427 -0.0238 0.0353 -0.9982>}
 ball_r {parent=foot_r, loc=<3.5658 -10.8066 0.1544>, rot=<0.0001 -0.0001 -0.7063 -0.7079>}
 calf_twist_02_r {parent=calf_r, loc=<19.1546 1.7943 0.2638>, rot=<0.0237 0.0095 -0.0035 -0.9997>}
 calf_twist_01_r {parent=calf_r, loc=<35.8483 1.9197 0.5794>, rot=<0.0237 0.0095 -0.0035 -0.9997>}
 root_ctrl {controlType=EULER_TRANSFORM, type=Control, loc=<0 0 0>, rot=<0 0 0 1>}
 foot_l_ctrl {controlType=EULER_TRANSFORM, parent=root_ctrl, type=Control, loc=<0 0 0>, rot=<0 0 0 1>}
 calf_twist_01_l_ctrl {controlType=EULER_TRANSFORM, parent=foot_l_ctrl, type=Control, loc=<14.3975 -5.8726 19.0937>, rot=<-0.0496 -0.727 0.0133 0.6847>}
 foot_r_ctrl {controlType=EULER_TRANSFORM, parent=root_ctrl, type=Control, loc=<0 0 0>, rot=<0 0 0 1>}
 calf_twist_01_r_ctrl {controlType=EULER_TRANSFORM, parent=foot_r_ctrl, type=Control, loc=<0 0 0>, rot=<0 0 0 1>}
]
